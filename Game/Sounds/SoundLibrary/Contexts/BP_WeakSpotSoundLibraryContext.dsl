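; Sound library context for weak spot sounds, linking an owner and an instigator actor to sound library events.
; Setters store both actors, and GatherActorTargets appends them to the outgoing target list.
; Play takes the sound and owner from the event input and, if both are valid, spawns the sound at that owner's location.
; It returns the spawned audio component in an array with success, or false otherwise.
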(fn GatherActorTargets
  (params (OutActors :type Array%20of%20Actor%20Object%20References))
  (Utilities|Array|AppendArray OutActors (Utilities|Array|MakeArray (Variables|Default|GetOwner) (Variables|Default|GetInstigator)))
  (return true))

(fn Play
  (params (InEventData :type Sound%20Library%20Context%20Event%20Input%20Structure%20%28by%20ref%29))
  (bind (_sound _owner) (Utilities|Struct|BreakSoundLibraryContextEventInput InEventData))
  (if (and (Utilities|IsValid _sound :!class /Script/Engine.KismetSystemLibrary) (Utilities|IsValid _owner :!class /Script/Engine.KismetSystemLibrary))
    (bind _returnvalue (Audio|SpawnSoundatLocation _sound (Transformation|GetActorLocation _owner)))
    (return (Utilities|Array|MakeArray _returnvalue) true)
    (else
      (return 0 false))))

(fn SetInstigator
  (params (Instigator :type Actor%20Object%20Reference))
  (Variables|Default|SetInstigator Instigator))

(fn SetOwner
  (params (Owner :type Actor%20Object%20Reference))
  (Variables|Default|SetOwner Owner))
